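; Quest gameplay ability that manages the cosmetic "UnseenDisguise" for a Season 24 narrative quest, ending once the quest is completed.
; Activation fires the disguise swap gameplay cue, then waits for the NPC-interaction tag to be added and removed.
; After a short delay, it checks the quest manager and ends itself if the quest is complete.
; On end it replays the cue, removes effects granting the Equipped tag and prints a debug string.
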(event Ability|EventActivateAbility
  ;[0 0 0] Check if when the player interacts with an NPC, should this ability be removed (is quest completed?)
  (|CallParentFunction)
  (Ability|ExecuteGameplayCueOnOwner (Variables|Default|GetDisguiseSwapTag))
  (Ability|Tasks|WaitGameplayTagAdd "(TagName=\"Conversation.InteractingWithNPC\")"
    (:Added
      (Ability|Tasks|WaitGameplayTagRemove "(TagName=\"Conversation.InteractingWithNPC\")"
        (:Removed
          (Utilities|FlowControl|Delay 2.5)
          (bind _bcompletedquest (CallFunction|HasPlayerCompletedQuest))
          (if _bcompletedquest
            (Ability|EndAbility)))))))

(event Ability|EventOnEndAbility
  (params (bWasCancelled :type Boolean))
  ;[0 0 0] Remove the pack on ability end
  (|CallParentFunction bWasCancelled)
  (Development|PrintString "Pack ability ending..." true true "(R=1.000000,G=0.000000,B=0.562280,A=1.000000)" 4.0)
  (Ability|ExecuteGameplayCueOnOwner (Variables|Default|GetDisguiseSwapTag))
  (GameplayEffects|RemoveActiveEffectswithGrantedTags (Ability|GetFortAbilitySystemComponent (Ability|GetAvatarActorfromActorInfo)) "(GameplayTags=((TagName=\"Athena.Quests.S24.UnseenDisguise.Equipped\")))"))

(fn HasPlayerCompletedQuest
  (return (Quest|HasCompletedQuest (Quests|GetQuestManager (Fortnite|GetFortPlayerControllerfromActor (CallFunction|GetAvatarasFortPlayerPawn)) "Invalid") (Variables|Default|GetQuestDefinition) :!class /Script/FortniteGame.FortQuestManager)))
